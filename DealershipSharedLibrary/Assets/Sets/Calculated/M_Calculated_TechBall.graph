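# Translucent lit material for Niagara mesh particles, apparently a calculated tech-ball effect.
# Pivot Painter per-object data lets Dynamic Parameters push mesh pieces along their X axis and, behind a scalar toggle, scale them toward their pivots.
# Particle color feeds an emissive adjustment function; the added reflection vector is multiplied by zero.
# Particle alpha minus vertex red, scaled by Fade, staggers opacity, with vertex alpha blending the Opacity parameter toward one.
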Material /DealershipSharedLibrary/Assets/Sets/Calculated/M_Calculated_TechBall
graph {
 properties {bUsedWithNiagaraMeshParticles=true, blendMode=BLEND_Translucent, usedShadingModels=Default Lit}
 nodes [
  n1 = DynamicParameter() {paramNames [Opacity,Push,Fade,Param4]}
  n2 = MaterialFunctionCall() {materialFunction=/Engine/Functions/Engine_MaterialFunctions02/PivotPainter/PivotPainter_PerObjectData}
  n3 = Multiply(A=@n1.Push, B=@n2.X-Axis Vector)
  n4 = Subtract(A=@n2.Pivot Position, B=WorldPosition().XYZ)
  n5 = OneMinus(None=DynamicParameter() {paramNames [Opacity,Push,Fade,Scale]}.Scale)
  If(
   A=Ceil(None=ScalarParameter() {defaultValue=1, parameterName="PivotPositionScale=1"})
   "A > B"=Add(A=@n3, B=Multiply(A=@n4, B=@n5))
   "A == B"=@n3
   "A < B"=@n3
  )
  n6 = ReflectionVectorWS()
  n7 = Constant()
  n8 = ParticleColor()
  MaterialFunctionCall(Emissive=Add(A=Multiply(A=@n6, B=@n7), B=@n8.RGB)) {materialFunction=&MF_FX_EmissiveAdjustments}
  n9 = VertexColor()
  n10 = Subtract(A=@n8.A, B=@n9.R)
  Multiply(A=LinearInterpolate(A=@n1.Opacity, B=Constant() {r=1}, Alpha=@n9.A), B=Saturate(None=Multiply(A=@n10, B=@n1.Fade)))
 ]
}
